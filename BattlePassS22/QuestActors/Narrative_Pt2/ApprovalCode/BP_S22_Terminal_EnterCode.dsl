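; Season 22 quest terminal Blueprint that shows a five-digit code on a decal as the player enters it through a conversation.
; The construction script creates the decal's dynamic material and writes each CodeDigits character into a digit parameter, plus a preview typing state.
; Client conversation messages carry a metadata integer that drives "Current Typing": blank, digits revealed, or accepted.
; Only local controllers, including splitscreen, update it; switch branches only print debug text.
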(event OnReceivedLocalClientConversationMessage(NonPlayerConversationComponent)
  (params (LocalPlayerController :type Fort%20Player%20Controller%20Object%20Reference) (Message_Message :type Client%20Conversation%20Message%20Structure) (Message_Participants :type Conversation%20Participants%20Structure) (Message_CurrentNode :type Conversation%20Node%20Handle%20Structure) (Message_Options :type Array%20of%20Client%20Conversation%20Option%20Entry%20Structures))
  (bind (_speakerid _participantdisplayname _text _metadataparameters) (Utilities|Struct|BreakClientConversationMessage Message_Message))
  (bind _output (Utilities|Array|Get(aref) _metadataparameters))
  (bind (_name _value) (Utilities|Struct|BreakConversationNodeParameterPair _output))
  (bind _returnvalue (Utilities|String|GetSubstring _value))
  (bind _returnvalue_2 (Utilities|String|StringToInteger _returnvalue))
  (bind _output_2 (CallFunction|IsValidLocalPlayerController LocalPlayerController))
  (if _output_2
    (if (not (Utilities|String|IsEmpty _returnvalue))
      ;[0 0 0] Message 0: Incorrect number entered or nothing should show on the screen. If the message number entered is 1, then show the first number in the decal; if 2, show both the first and second, etc. 5 numbers total where 6 is the "accepted/finished" state
      (switch int _returnvalue_2
        (:0
          ;;[0.1667 0.1667 0.1667] Set decal parameters here
          (Development|PrintString "Empty" true true "(R=0.295317,G=1.000000,B=0.000000,A=1.000000)"))
        (:1
          (Development|PrintString (Utilities|String|ToString(Integer) _returnvalue_2) true true "(R=0.295317,G=1.000000,B=0.000000,A=1.000000)"))
        (:2
          _)
        (:3
          _)
        (:4
          _)
        (:5
          _)
        (:6
          (Development|PrintString "Accepted" true true "(R=0.295317,G=1.000000,B=0.000000,A=1.000000)")))
      (Rendering|Material|SetScalarParameterValue (Variables|Default|GetDigitsMID) "Current Typing" (Math|Conversions|ToFloat(Integer) _returnvalue_2) :!class /Script/Engine.MaterialInstanceDynamic))
    (else
      (Development|PrintString "Invalid" true true "(R=1.000000,G=0.000000,B=0.149854,A=1.000000)" 3.0))))

(fn IsValidLocalPlayerController
  (params (InstigatingPlayerController :type Fort%20Player%20Controller%20Object%20Reference))
  ;[0 0 0] Returns true on the executing client if the instigator controller is locally controlled. Returns false if running on the server
  (if (not (Networking|IsServer))
    (Utilities|IsValid InstigatingPlayerController
      (:Is%20Valid
        (if (PlayerController|IsinSplitscreen InstigatingPlayerController)
          (for _array_element (Fortnite|GetLocalFortPlayerControllers true)
            (if (== _array_element InstigatingPlayerController)
              (return true)))
          (return false)
          (else
            (return true))))
      (:Is%20Not%20Valid
        _)))
  (return false))

(fn ConstructionScript
  (bind _codedigits (Variables|Default|GetCodeDigits))
  (|CallParentFunction)
  (bind _returnvalue (Rendering|Components|Decal|CreateDynamicMaterialInstance (Variables|Default|GetDigitsDecal)))
  (bind _output_get (Variables|Default|SetDigitsMID _returnvalue))
  (Rendering|Material|SetScalarParameterValue _output_get "digit0" (Math|Conversions|ToFloat(Integer) (- (Utilities|String|GetCharacterasNumber _codedigits) 48)) :!class /Script/Engine.MaterialInstanceDynamic)
  (Rendering|Material|SetScalarParameterValue _output_get "digit1" (Math|Conversions|ToFloat(Integer) (- (Utilities|String|GetCharacterasNumber _codedigits 1) 48)) :!class /Script/Engine.MaterialInstanceDynamic)
  (Rendering|Material|SetScalarParameterValue _output_get "digit2" (Math|Conversions|ToFloat(Integer) (- (Utilities|String|GetCharacterasNumber _codedigits 2) 48)) :!class /Script/Engine.MaterialInstanceDynamic)
  (Rendering|Material|SetScalarParameterValue _output_get "digit3" (Math|Conversions|ToFloat(Integer) (- (Utilities|String|GetCharacterasNumber _codedigits 3) 48)) :!class /Script/Engine.MaterialInstanceDynamic)
  (Rendering|Material|SetScalarParameterValue _output_get "digit4" (Math|Conversions|ToFloat(Integer) (- (Utilities|String|GetCharacterasNumber _codedigits 4) 48)) :!class /Script/Engine.MaterialInstanceDynamic)
  (Rendering|Material|SetScalarParameterValue _output_get "Current Typing" (Math|Conversions|ToFloat(Integer) (Variables|Default|GetPreview%20Code)) :!class /Script/Engine.MaterialInstanceDynamic))
