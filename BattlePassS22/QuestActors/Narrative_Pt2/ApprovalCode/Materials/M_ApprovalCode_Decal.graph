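# Deferred decal material for a five-slot approval code display, showing digits as they are typed with a blinking underscore cursor.
# Per-slot digit parameters and a Current Typing index select frames from an 11-row CodeDigits flipbook, with the last frame as the cursor.
# Panning static and a time-driven sweep line build glitch UV distortion, and BumpOffset adds parallax.
# A sine-based band trims the glyph mask, alongside a Screen Tone color parameter.
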Material /BattlePassS22/QuestActors/Narrative_Pt2/ApprovalCode/Materials/M_ApprovalCode_Decal
graph {
 properties {blendMode=BLEND_Translucent, materialDomain=MD_DeferredDecal, pixelDepthOffsetMode=PDOM_Legacy, usedShadingModels=Default Lit}
 nodes [
  n1 = Constant() {r=6.2832}
  n2 = Constant()
  n3 = Rotator(
   Coordinate=NamedRerouteUsage()
   Time=ChannelMaskParameter(None=AppendVector(A=@n1, B=@n2)) {maskChannel=Green, parameterName=Glitch Direction, defaultValue=<rgba 0 1 0 0>}.R
  )
  n4 = Panner(Coordinate=@n3) {speedX=6, speedY=-0.05}
  n5 = Panner(Coordinate=@n3) {speedX=-3, speedY=-0.03}
  n6 = Multiply(
   A=Multiply(
    A=TextureSample(UVs=@n4) {samplerType=SAMPLERTYPE_Masks, texture=/Game/Missions/Primary/RepairTheShelter/Textures/Cinematic_Screen_Static}.G
    B=TextureSample(UVs=@n5) {samplerType=SAMPLERTYPE_Masks, texture=&Cinematic_Screen_Static}.R
   )
   B=ScalarParameter() {defaultValue=0.5, group=Static, parameterName=Static UV Distort}
  )
  n7 = Constant()
  n8 = Multiply(A=ChannelMaskParameter(None=NamedRerouteUsage()) {maskChannel=Green, parameterName=Glitch Direction, defaultValue=<rgba 0 1 0 0>}.R)
  n9 = Multiply(
   A=Add(A=Time(), B=ScalarParameter() {defaultValue=0.3, parameterName=Offset}) {constB=0.5}
   B=ScalarParameter() {defaultValue=1, parameterName=Speed}
  ) {constB=-0.2}
  n10 = Subtract(A=Frac(None=Add(A=@n8, B=@n9)), B=ScalarParameter() {defaultValue=0.8, parameterName=Dist Bias})
  n11 = Multiply(A=Saturate(None=Sign(None=@n10)), B=ScalarParameter() {defaultValue=0.005, group=Sweep Line, parameterName=Sweep UV Distort}) {constB=0.02}
  n12 = ComponentMask(None=VectorParameter() {parameterName=Distortion Direction, defaultValue=<rgba 1 0.7556 0 1>}.RGB) {g=true, r=true}
  NamedRerouteDeclaration(None=Add(A=AppendVector(A=@n6, B=@n7), B=Multiply(A=@n11, B=@n12))) {name=Distorted UVs, nodeColor=<rgba 0 0.3203 1 1>}
  n13 = ScalarParameter() {defaultValue=4, parameterName=Digit4}
  n14 = Floor(None=ComponentMask(None=TextureCoordinate() {uTiling=5}) {r=true})
  n15 = If(A=@n14, "A > B"=@n13, "A == B"=ScalarParameter() {parameterName=Digit0}, "A < B"=@n13)
  n16 = If(A=@n14, "A > B"=@n15, "A == B"=ScalarParameter() {defaultValue=1, parameterName=Digit1}, "A < B"=@n15) {constB=1}
  n17 = If(A=@n14, "A > B"=@n16, "A == B"=ScalarParameter() {defaultValue=2, parameterName=Digit2}, "A < B"=@n16) {constB=2}
  NamedRerouteDeclaration(None=If(A=@n14, "A > B"=@n17, "A == B"=ScalarParameter() {defaultValue=3, parameterName=Digit3}, "A < B"=@n17) {constB=3}) {name=Digits, nodeColor=<rgba 1 0 0.5156 1>}
  n18 = If(
   A=@n14
   B=Floor(None=ScalarParameter() {parameterName=Current Typing})
   "A > B"=Constant()
   "A == B"=Constant() {r=-1}
   "A < B"=Constant() {r=1}
  )
  NamedRerouteDeclaration(None=Saturate(None=@n18)) {name=Digit Mask, nodeColor=<rgba 0.5469 0 1 1>}
  n19 = Multiply(
   A=TextureCoordinate()
   B=ComponentMask(None=VectorParameter() {parameterName=TexCoordScale, defaultValue=<rgba 1 1 0 0>}.RGB) {g=true, r=true}
  )
  n20 = ComponentMask(None=VectorParameter() {parameterName=TexCoordOffset, defaultValue=<rgba -0.5 0 0 0>}.RGB) {g=true, r=true}
  NamedRerouteDeclaration(None=BumpOffset(Coordinate=Add(A=@n19, B=@n20), Height=ScalarParameter() {defaultValue=-5, parameterName=Bump Height})) {name=TexCoord, nodeColor=<rgba 1 0 0.4453 1>}
  n21 = Multiply(A=@n18) {constB=-1}
  NamedRerouteDeclaration(None=Multiply(A=Saturate(None=@n21)) {constB=10}) {name=Underscore Index, nodeColor=<rgba 0 1 0.3672 1>}
  n22 = NamedRerouteUsage()
  n23 = NamedRerouteUsage()
  n24 = MaterialFunctionCall(
   "Animation  Phase (0-1)"=Divide(A=Max(A=@n22, B=@n23)) {constB=10}
   Number of Rows=Constant() {r=11}
   Number of Columns=Constant() {r=1}
   Texture=TextureObject() {texture=/BattlePassS22/QuestActors/Narrative_Pt2/ApprovalCode/Textures/CodeDigits}
   UVs=TextureCoordinate() {uTiling=5, vTiling=5}
   "Clamp Anim (see tooltip)"=StaticBool() {value=true}
  ) {materialFunction=&FlipBook}
  n25 = MaterialFunctionCall(Y=ScalarParameter() {defaultValue=0.82}) {materialFunction=&MakeFloat2}
  n26 = TextureSample(UVs=Add(A=Add(A=@n24.UVs, B=@n25.Result), B=NamedRerouteUsage())) {texture=&CodeDigits}
  n27 = ComponentMask(None=TextureCoordinate() {vTiling=0.5}) {g=true}
  n28 = Sign(None=Subtract(A=Sine(None=@n27), B=ScalarParameter() {defaultValue=0.95, parameterName=Param_1}))
  n29 = NamedRerouteUsage()
  n30 = Round(None=Frac(None=Time()))
  n31 = NamedRerouteUsage()
  NamedRerouteDeclaration(None=Multiply(A=Multiply(A=@n26.R, B=@n28), B=LinearInterpolate(A=@n29, B=@n30, Alpha=@n31))) {name=Mask, nodeColor=<rgba 0 1 0.9063 1>}
  NamedRerouteUsage()
  VectorParameter() {group=Colors, parameterName=Screen Tone, defaultValue=<rgba 0.1292 0.2135 0.069 1>}
 ]
}
